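# WBP_GenericInventoryScreen_Container is a Widget Blueprint that serves as a container for a generic inventory screen.
# It derives from the native GenericInventoryScreenContainer class in the ItemizationUI module.
# Its only widget is the inherited Content_InventoryScreen named slot, which is the root and holds no content here.
# The Blueprint adds no variables or graph logic of its own.
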WidgetBlueprint /Itemization/Widgets/WBP_GenericInventoryScreen_Container
graphs [EventGraph {links [], nodes []}]
variables []
widgets {
 info {
  inheritedWidgetCount=1
  parentClass=/Script/ItemizationUI.GenericInventoryScreenContainer
  rootWidgetClass=&NamedSlot
  widgetCount=1
 }
 widgets [Content_InventoryScreen {widgetClassPath=&NamedSlot, bInherited=true, widget=Content_InventoryScreen}]
}
