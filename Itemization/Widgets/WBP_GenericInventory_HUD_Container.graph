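# WBP_GenericInventory_HUD_Container is a widget blueprint that serves as the HUD container for Itemization's generic inventory UI, built on the native GenericInventoryHUDContainer class.
# Its root HUD element wrapper is tagged Itemization.UI.HUD.Container and hosts an inherited Content_HUDWidget named slot for the inventory content.
# The wrapper carries gameplay tags that hide it during states like loading, map, vehicle, bus, post-game and special events.
# The blueprint has no graph logic of its own.
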WidgetBlueprint /Itemization/Widgets/WBP_GenericInventory_HUD_Container
graphs [EventGraph {links [], nodes []}]
variables []
widgets {
 info {
  inheritedWidgetCount=1
  namedSlotCount=1
  parentClass=/Script/ItemizationUI.GenericInventoryHUDContainer
  rootWidgetClass=/Game/Athena/HUD/FortHUDElementWrapper.FortHUDElementWrapper_C
  widgetCount=2
 }
 widgets [
  HUDElementWrapper {
   widgetClassPath=/Game/Athena/HUD/FortHUDElementWrapper
   widget=HUDElementWrapper
   properties {
    elementTag {tagName=Itemization.UI.HUD.Container}
    hUDElementTag {gameplayTags [
      {tagName=Athena.Creative.HideForLoading}
      {tagName=Athena.Creative.HUDController.HidePlayerInventory}
      {tagName=Athena.HideForCapture}
      {tagName=HUD.Athena.Hide.WithInventory}
      {tagName=HUD.Athena.Hide.WithMap}
      {tagName=HUD.Athena.Hide.WithRCPawn}
      {tagName=HUD.Athena.HideForAbility}
      {tagName=HUD.Athena.HideForConversation}
      {tagName=HUD.Athena.HideForGadget}
      {tagName=HUD.Athena.HideForPostGame}
      {tagName=HUD.Athena.HideForPostGameSpectate}
      {tagName=HUD.Athena.HideforRadial}
      {tagName=HUD.Athena.HideforVehicle}
      {tagName=HUD.Athena.HideOnBus}
      {tagName=HUD.Event.HideForSpecialEvent}
      {tagName=HUD.Hide.CreativeJoiningLobby}
      {tagName=HUD.Hide.WhileLoading}
     ]}
   }
  }
  Content_HUDWidget {widgetClassPath=&NamedSlot, bInherited=true, namedSlotHost=HUDElementWrapper, widget=Content_HUDWidget}
 ]
}
